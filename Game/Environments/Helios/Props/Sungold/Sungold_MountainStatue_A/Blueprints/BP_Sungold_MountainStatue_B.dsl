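; BP_Sungold_MountainStatue_B is a Blueprint for a Sungold mountain statue prop that carries a replicable crack state.
; The Set Cracks Enabled event stores its boolean into the Crack Enabled variable, which has an OnRep function.
; That OnRep function, BeginPlay and Validate Statue Cracks are empty here, and the construction script only calls its parent.
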(event EventBeginPlay)

(event Custom|Set%20Cracks%20Enabled
  (params (Crack%20Enabled :type Boolean))
  ;[1 1 1] Begin Play / Cracks
  (Variables|Default|SetCrack%20Enabled Crack%20Enabled))

(event Custom|Validate%20Statue%20Cracks)

(fn OnRep_CrackEnabled)

(fn ConstructionScript
  (|CallParentFunction))
